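; Editor actor that bakes the sky across a day cycle into the Sky Atmos LUT volume texture.
; The capture event steps Day Sequence preview time per slice, captures the scene in 2D and passes the slice index to Niagara, presumably to fill the volume render target.
; Frame delays separate steps, and slice 0 comes last at hour 24.
; It then converts the volume to the texture, restores compression settings and resets render target sizes.
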(event Custom|Sky%20Atmos%20Capture
  (bind _sky_atmos_lut (Variables|Default|GetSky%20Atmos%20LUT))
  (Utilities|FlowControl|DelayUntilNextFrame)
  (Rendering|ResizeRenderTarget2D (Variables|Default|GetRenderTarget2D) (Variables|Default|GetSize%20X) (Variables|Default|GetSize%20Y))
  (Variables|/Script/Engine.TextureRenderTargetVolume|SetSizeX (Variables|Default|GetSize%20X) (Variables|Default|GetRenderTarget3D))
  (Variables|/Script/Engine.TextureRenderTargetVolume|SetSizeY (Variables|Default|GetSize%20Y) (Variables|Default|GetRenderTarget3D))
  (Variables|/Script/Engine.TextureRenderTargetVolume|SetSizeZ (Variables|Default|GetSize%20Z) (Variables|Default|GetRenderTarget3D))
  (Utilities|FlowControl|DelayUntilNextTick)
  (Variables|Default|SetSlicesLoop 1)
  (Variables|Default|SetCompression%20Settings (Variables|/Script/Engine.Texture|GetCompressionSettings _sky_atmos_lut))
  (Variables|/Script/Engine.Texture|SetDeferCompression true _sky_atmos_lut)
  (Variables|/Script/Engine.Texture|SetCompressionNoAlpha true _sky_atmos_lut)
  (CallFunction|SetTimeOfDay (Variables|Default|GetSlicesLoop) (Variables|Default|GetSize%20Z))
  (Utilities|FlowControl|DelayUntilNextFrame)
  (CallFunction|SetRenderTarget (Variables|Default|GetSlicesLoop))
  (Utilities|FlowControl|DelayUntilNextFrame)
  (bind _output_get (Variables|Default|SetSlicesLoop (+ (Variables|Default|GetSlicesLoop) 1)))
  (if (== _output_get (Variables|Default|GetSize%20Z))
    _
    (else
      _))
  (CallFunction|SetTimeOfDay (Variables|Default|GetSlicesLoop) (Variables|Default|GetSize%20Z))
  (Utilities|FlowControl|DelayUntilNextFrame)
  (CallFunction|SetRenderTarget 0)
  (Utilities|FlowControl|DelayUntilNextFrame)
  (EditorScripting|Rendering|ConvertRenderTargettoTextureVolumeEditorOnly (Variables|Default|GetRenderTarget3D) (Variables|Default|GetSky%20Atmos%20LUT))
  (Variables|/Script/Engine.Texture|SetCompressionSettings (Variables|Default|GetCompression%20Settings) (Variables|Default|GetSky%20Atmos%20LUT))
  (Utilities|FlowControl|DelayUntilNextFrame)
  (Rendering|ResizeRenderTarget2D (Variables|Default|GetRenderTarget2D))
  (Variables|/Script/Engine.TextureRenderTargetVolume|SetSizeX 256 (Variables|Default|GetRenderTarget3D))
  (Variables|/Script/Engine.TextureRenderTargetVolume|SetSizeY 256 (Variables|Default|GetRenderTarget3D))
  (Variables|/Script/Engine.TextureRenderTargetVolume|SetSizeZ 256 (Variables|Default|GetRenderTarget3D)))

(fn SetRenderTarget
  (params (Index :type Integer))
  (Rendering|SceneCapture|CaptureScene (Variables|Default|GetSceneCaptureComponent2D) :!class /Script/Engine.SceneCaptureComponent2D)
  (Niagara|SetNiagaraVariable(Int32) (Variables|Default|GetNiagara) "User.TextureHeight" (Variables|GetIndex)))

(fn SetTimeOfDay
  (params (Index :type Integer) (TotalSlices :type Integer))
  (TimeOfDay|SetTimeOfDayPreview (DaySequence|GetDaySequenceActor (WorldSubsystems|GetDaySequenceSubsystem) true) (* Index (/ 24.0 TotalSlices))))

(fn ConstructionScript
  (Rendering|ResizeRenderTarget2D (Variables|Default|GetRenderTarget2D) (Variables|Default|GetSize%20X) (Variables|Default|GetSize%20Y))
  (Variables|/Script/Engine.TextureRenderTargetVolume|SetSizeX (Variables|Default|GetSize%20X) (Variables|Default|GetRenderTarget3D))
  (Variables|/Script/Engine.TextureRenderTargetVolume|SetSizeY (Variables|Default|GetSize%20Y) (Variables|Default|GetRenderTarget3D))
  (Variables|/Script/Engine.TextureRenderTargetVolume|SetSizeZ (Variables|Default|GetSize%20Z) (Variables|Default|GetRenderTarget3D))
  (Transformation|SetRelativeScale3D (Variables|Default|GetPlane) (Math|Vector|MakeVector 1.0 (/ (Math|Conversions|ToFloat(Integer) (Variables|Default|GetSize%20Y)) (Math|Conversions|ToFloat(Integer) (Variables|Default|GetSize%20X))) 1.0)))
