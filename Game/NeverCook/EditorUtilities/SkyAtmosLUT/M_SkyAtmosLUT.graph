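# Unlit editor-utility material that outputs Sky Atmosphere view luminance as a lat-long map, by its name and folder for a sky atmosphere LUT.
# UVs are scaled to half height in V, then converted to a world direction by UVToLongLat, so the map spans only half the latitude range.
# That direction feeds SkyAtmosphereViewLuminance, which samples the sky color per texel.
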Material /Game/NeverCook/EditorUtilities/SkyAtmosLUT/M_SkyAtmosLUT
graph {
 properties {
  bAutomaticallySetUsageInEditor=false
  bDisableDepthTest=true
  bUseTranslucencyVertexFog=false
  shadingModel=MSM_Unlit
  translucencyPass=MTP_BeforeDOF
  usedShadingModels=Unlit
 }
 nodes [
  n1 = TextureCoordinate()
  n2 = Constant2Vector() {g=0.5, r=1}
  SkyAtmosphereViewLuminance!(WorldDirection=MaterialFunctionCall!(UV=Multiply(A=@n1, B=@n2)) {materialFunction=/Engine/Functions/Engine_MaterialFunctions01/Coordinates/UVToLongLat}.Result)
 ]
}
