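# Editor-only Niagara system under NeverCook, named as a utility for sky atmosphere LUTs, with default targets RT_SkyAtmosLUT_2D and RT_SkyAtmosLUT_3D.
# A single renderer-less GPU emitter creates 2D and volume render target data interfaces at spawn, bound to user texture parameters and inheriting their settings.
# The shown stack has only stock state and particle initialization modules; nothing visible writes to the targets or reads the TextureHeight parameter.
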NiagaraSystem /Game/NeverCook/EditorUtilities/SkyAtmosLUT/NS_SkyAtmosLUT
dependencies {
 usedDataInterfaces [/Script/Niagara.NiagaraDataInterfaceRenderTarget2D, /Script/Niagara.NiagaraDataInterfaceRenderTargetVolume]
 usedDynamicInputs []
 usedModules [&SystemState]
}
emitters {Minimal {
  simTarget=GPUComputeSim
  emitterSpawnScript [
   /SetVariables(Emitter.Render Target 2D=DataInterface() {propertyValues {
      MipMapGeneration=Disabled
      MipMapGenerationType=Linear
      OverrideRenderTargetFilter=TF_Default
      OverrideRenderTargetFormat=RTF_R8
      bInheritUserParameterSettings=true
      bOverrideFormat=false
      bPreviewRenderTarget=false
      RenderTargetUserParameter {parameter {name=User.Texture Render Target 2D, typeDefHandle {registeredTypeIndex=4}, varData []}}
      Size=<0 0>
     }}) {isSetParameters=true}
   /SetVariables(Emitter.Render Target 3D=DataInterface() {propertyValues {
      OverrideRenderTargetFilter=TF_Default
      OverrideRenderTargetFormat=RTF_R8
      bInheritUserParameterSettings=true
      bOverrideFormat=false
      bPreviewRenderTarget=false
      RenderTargetUserParameter {parameter {name=User.Texture Render Target 3D, typeDefHandle {registeredTypeIndex=4}, varData []}}
      Size=<0 0 0>
     }}) {isSetParameters=true}
  ]
  emitterUpdateScript [&EmitterState()]
  particleSpawnScript [&InitializeParticle(Color Mode=::NewEnumerator0, Lifetime=5, UsePositionOffset=false, Write Material Random=true)]
  particleUpdateScript [&ParticleState()]
  rendererClasses []
  renderers []
 }}
summary {
 systemName=NS_SkyAtmosLUT
 emitters [{emitterName=Minimal, simTarget=GPUComputeSim, rendererClasses []}]
 userVariables [
  User.Texture Render Target 2D {
   type=TextureRenderTarget
   description=""
   defaultValue {
    struct=NiagaraExt_VariableValue_Object
    value {object=&/RT_SkyAtmosLUT_2D, objectClass=/Script/Engine.TextureRenderTarget}
   }
  }
  User.Texture Render Target 3D {
   type=TextureRenderTarget
   description=""
   defaultValue {
    struct=NiagaraExt_VariableValue_Object
    value {object=&/RT_SkyAtmosLUT_3D, objectClass=&TextureRenderTarget}
   }
  }
  User.TextureHeight {type=NiagaraInt32}
 ]
}
